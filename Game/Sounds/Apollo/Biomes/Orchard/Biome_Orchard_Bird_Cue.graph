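# Sound cue for one-shot bird calls in the Orchard biome's outdoor ambience, built from Farm biome BirdB recordings.
# A Random node picks one of eleven equally weighted wave players and feeds a Modulator that slightly varies pitch and volume on each play.
# The cue plays at reduced volume through a tile-ambient one-shot attenuation and the Athena outdoor ambient sound class.
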SoundCue /Game/Sounds/Apollo/Biomes/Orchard/Biome_Orchard_Bird_Cue
root=Modulator_0
nodes [
 {id=Modulator_0, children [Random_2], properties {pitchMax=1.1, pitchMin=0.9, volumeMax=1.2, volumeMin=0.8}}
 {
  id=Random_2
  children [
   WavePlayer_0
   WavePlayer_1
   WavePlayer_2
   WavePlayer_3
   WavePlayer_4
   WavePlayer_5
   WavePlayer_7
   WavePlayer_8
   WavePlayer_10
   WavePlayer_11
   WavePlayer_12
  ]
  properties {weights [1,1,1,1,1,1,1,1,1,1,1]}
 }
 {id=WavePlayer_0, properties {soundWaveAssetPtr=/Game/Sounds/Apollo/Biomes/Farm/Biome_Farm_BirdB_10}}
 {id=WavePlayer_1, properties {soundWaveAssetPtr=/Game/Sounds/Apollo/Biomes/Farm/Biome_Farm_BirdB_01}}
 {id=WavePlayer_10, properties {soundWaveAssetPtr=/Game/Sounds/Apollo/Biomes/Farm/Biome_Farm_BirdB_04}}
 {id=WavePlayer_11, properties {soundWaveAssetPtr=/Game/Sounds/Apollo/Biomes/Farm/Biome_Farm_BirdB_06}}
 {id=WavePlayer_12, properties {soundWaveAssetPtr=/Game/Sounds/Apollo/Biomes/Farm/Biome_Farm_BirdB_08}}
 {id=WavePlayer_2, properties {soundWaveAssetPtr=/Game/Sounds/Apollo/Biomes/Farm/Biome_Farm_BirdB_03}}
 {id=WavePlayer_3, properties {soundWaveAssetPtr=/Game/Sounds/Apollo/Biomes/Farm/Biome_Farm_BirdB_05}}
 {id=WavePlayer_4, properties {soundWaveAssetPtr=/Game/Sounds/Apollo/Biomes/Farm/Biome_Farm_BirdB_07}}
 {id=WavePlayer_5, properties {soundWaveAssetPtr=/Game/Sounds/Apollo/Biomes/Farm/Biome_Farm_BirdB_09}}
 {id=WavePlayer_7, properties {soundWaveAssetPtr=/Game/Sounds/Apollo/Biomes/Farm/Biome_Farm_BirdB_11}}
 {id=WavePlayer_8, properties {soundWaveAssetPtr=/Game/Sounds/Apollo/Biomes/Farm/Biome_Farm_BirdB_02}}
]
properties {
 attenuationSettings=/Game/Sounds/Attenuation/Fort_TileAmbient_OneShot_Att
 maxDistance=8000
 soundClassObject=&Fort_Ambient_Outdoors_Athena
 volumeMultiplier=0.451
 preEffectBusSends [{}]
}
